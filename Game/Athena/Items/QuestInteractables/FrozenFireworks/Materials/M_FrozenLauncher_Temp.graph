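# Temporary Default Lit material for the Frozen Fireworks launcher quest interactable, previewed on the launcher mesh.
# It samples the launcher's UV template texture and uses the green channel to blend from white to red.
# The file shows no link from that blend to a material output, so the graph appears to be an unfinished placeholder.
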Material /Game/Athena/Items/QuestInteractables/FrozenFireworks/Materials/M_FrozenLauncher_Temp
graph {
 properties {
  pixelDepthOffsetMode=PDOM_Legacy
  previewMesh=/Game/Athena/Items/QuestInteractables/FrozenFireworks/Meshes/SM_FrozenFireworkLauncher
  usedShadingModels=Default Lit
 }
 nodes [LinearInterpolate(
   B=Constant3Vector() {constant=<rgba 1 0 0 0>}
   Alpha=TextureSample() {texture=/Game/Athena/Items/QuestInteractables/FrozenFireworks/Textures/T_UVTemplate_FrozenLauncher}.G
  ) {constA=1}]
}
